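# Sound cue for the end of an ammo crate search, spatialized with prop-search attenuation and limited by a stop-farthest concurrency override.
# A quality switch feeds two identical layered branches and two cheaper branches that each play one modulated, pre-optimized search-end wave.
# Layered branches mix the no-more-loot UI sound with randomized crate and ammunition foley, varied in pitch and volume and quickly faded out by a volume envelope.
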SoundCue /Figment_S10/Sounds/Foley_Loot/Containers/AmmoCrate/AmmoCrate_SearchEnd_Cue
root=QualityLevel_0
nodes [
 {
  id=Enveloper_0
  children [Mixer_1]
  properties {pitchCurve {editorCurveData {keys [{time=0}]}}, volumeCurve {editorCurveData {keys [<at 0.0045 1>,<at 0.3593 0.1392>,<at 0.7415 0>]}}}
 }
 {
  id=Enveloper_21
  children [Mixer_8]
  properties {pitchCurve {editorCurveData {keys [{time=0}]}}, volumeCurve {editorCurveData {keys [<at 0.0045 1>,<at 0.3592 0.1392>,<at 0.7415 0>]}}}
 }
 {id=Mixer_0, children [WavePlayer_43, Modulator_15], properties {inputVolume [1,1]}}
 {id=Mixer_1, children [Modulator_1,Modulator_2], properties {inputVolume [0.7,1]}}
 {id=Mixer_2, children [WavePlayer_0,Modulator_0], properties {inputVolume [1,1]}}
 {id=Mixer_8, children [Modulator_19,Modulator_20], properties {inputVolume [0.7,1]}}
 {id=Modulator_0, children [Enveloper_0], properties {pitchMax=1.07, pitchMin=0.92, volumeMax=1.07, volumeMin=0.92}}
 {id=Modulator_1, children [Random_0], properties {pitchMax=1.07, pitchMin=0.92, volumeMax=1.07, volumeMin=0.92}}
 {id=Modulator_15, children [Enveloper_21], properties {pitchMax=1.07, pitchMin=0.92, volumeMax=1.07, volumeMin=0.92}}
 {id=Modulator_19, children [Random_56], properties {pitchMax=1.07, pitchMin=0.92, volumeMax=1.07, volumeMin=0.92}}
 {id=Modulator_2, children [Random_1], properties {pitchMax=1.07, pitchMin=0.92, volumeMax=1.07, volumeMin=0.92}}
 {id=Modulator_20, children [Random_57], properties {pitchMax=1.07, pitchMin=0.92, volumeMax=1.07, volumeMin=0.92}}
 {id=Modulator_3, children [WavePlayer_16], properties {volumeMax=1.07, volumeMin=0.92}}
 {id=Modulator_4, children [WavePlayer_6], properties {pitchMax=1.07, pitchMin=0.92, volumeMax=1.07, volumeMin=0.92}}
 {id=QualityLevel_0, children [Mixer_0,Mixer_2,Modulator_3,Modulator_4]}
 {id=Random_0, children [WavePlayer_1,WavePlayer_2], properties {weights [1,1]}}
 {id=Random_1, children [WavePlayer_3,WavePlayer_4], properties {bRandomizeWithoutReplacement=false, weights [1,1]}}
 {id=Random_56, children [WavePlayer_355, WavePlayer_358], properties {weights [1,1]}}
 {id=Random_57, children [WavePlayer_360, WavePlayer_361], properties {bRandomizeWithoutReplacement=false, weights [1,1]}}
 {id=WavePlayer_0, properties {soundWaveAssetPtr=/Game/Sounds/Fort_UI_Sounds/Gameplay/Fort_UI_NoMoreLoot}}
 {id=WavePlayer_1, properties {soundWaveAssetPtr=&/AmmoCrate_Search10}}
 {id=WavePlayer_16, properties {soundWaveAssetPtr=&/Ammunition_SearchEnd_opt}}
 {id=WavePlayer_2, properties {soundWaveAssetPtr=&/AmmoCrate_Search08}}
 {id=WavePlayer_3, properties {soundWaveAssetPtr=&/Ammunition_Search10}}
 {id=WavePlayer_355, properties {soundWaveAssetPtr=&AmmoCrate_Search10}}
 {id=WavePlayer_358, properties {soundWaveAssetPtr=&AmmoCrate_Search08}}
 {id=WavePlayer_360, properties {soundWaveAssetPtr=&Ammunition_Search10}}
 {id=WavePlayer_361, properties {soundWaveAssetPtr=&/Ammunition_Search12}}
 {id=WavePlayer_4, properties {soundWaveAssetPtr=&Ammunition_Search12}}
 {id=WavePlayer_43, properties {soundWaveAssetPtr=&Fort_UI_NoMoreLoot}}
 {id=WavePlayer_6, properties {soundWaveAssetPtr=&Ammunition_SearchEnd_opt}}
]
properties {
 attenuationSettings=/Game/Sounds/Attenuation/Fort_Prop_Search_Att
 bOverrideConcurrency=true
 maxDistance=1300
 soundClassObject=/Figment_S10/Sounds/FortSoundClassesAndModes/Fort_Searching_Foley
 volumeMultiplier=0.7
 concurrencyOverrides {maxCount=4, resolutionRule=StopFarthestThenPreventNew}
}
