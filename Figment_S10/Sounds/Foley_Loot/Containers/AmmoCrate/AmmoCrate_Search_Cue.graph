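# Sound cue for searching an ammo crate loot container, routed to the searching foley class with prop search attenuation.
# A quality-level root picks between two layered mixes and, for the two lowest-cost slots, one pre-optimized wave.
# The mixes layer randomized ammo box, quieter ammunition and gear sounds, mostly pitch- and volume-modulated, with fewer variants in the second.
# Overridden concurrency prevents new instances at the limit.
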SoundCue /Figment_S10/Sounds/Foley_Loot/Containers/AmmoCrate/AmmoCrate_Search_Cue
root=QualityLevel_0
nodes [
 {id=Mixer_0, children [Modulator_0,Modulator_1,Modulator_2], properties {inputVolume [1,0.4,1]}}
 {id=Mixer_7, children [Modulator_17,Modulator_18,Random_8], properties {inputVolume [1,0.4,1]}}
 {id=Modulator_0, children [Random_3], properties {pitchMax=1.07, pitchMin=0.92, volumeMax=1.07, volumeMin=0.92}}
 {id=Modulator_1, children [Random_1], properties {pitchMax=1.07, pitchMin=0.92, volumeMax=1.07, volumeMin=0.92}}
 {id=Modulator_17, children [Random_0], properties {pitchMax=1.07, pitchMin=0.92, volumeMax=1.07, volumeMin=0.92}}
 {id=Modulator_18, children [Random_55], properties {pitchMax=1.07, pitchMin=0.92, volumeMax=1.07, volumeMin=0.92}}
 {id=Modulator_2, children [Random_2], properties {pitchMax=1.07, pitchMin=0.92, volumeMax=1.07, volumeMin=0.92}}
 {id=Modulator_3, children [WavePlayer_12]}
 {id=Modulator_4, children [WavePlayer_14]}
 {id=QualityLevel_0, children [Mixer_7,Mixer_0,Modulator_3,Modulator_4]}
 {id=Random_0, children [WavePlayer_0,WavePlayer_1,WavePlayer_2], properties {weights [1,1,1]}}
 {id=Random_1, children [WavePlayer_3,WavePlayer_4], properties {weights [1,1]}}
 {id=Random_2, children [WavePlayer_6,WavePlayer_7], properties {weights [1,1]}}
 {id=Random_3, children [WavePlayer_9, WavePlayer_10], properties {weights [1,1]}}
 {id=Random_55, children [WavePlayer_349, WavePlayer_350, WavePlayer_351], properties {weights [1,1,1]}}
 {id=Random_8, children [WavePlayer_37, WavePlayer_38, WavePlayer_41], properties {weights [1,1,1]}}
 {id=WavePlayer_0, properties {soundWaveAssetPtr=&/ammobox_B_search_01}}
 {id=WavePlayer_1, properties {soundWaveAssetPtr=&/ammobox_B_search_02}}
 {id=WavePlayer_10, properties {soundWaveAssetPtr=&ammobox_B_search_02}}
 {id=WavePlayer_12, properties {soundWaveAssetPtr=&/ammocrate_search_01_opt}}
 {id=WavePlayer_14, properties {soundWaveAssetPtr=&ammocrate_search_01_opt}}
 {id=WavePlayer_2, properties {soundWaveAssetPtr=&/ammobox_B_search_03}}
 {id=WavePlayer_3, properties {soundWaveAssetPtr=&/Ammunition_Search08}}
 {id=WavePlayer_349, properties {soundWaveAssetPtr=&Ammunition_Search08}}
 {id=WavePlayer_350, properties {soundWaveAssetPtr=&/Ammunition_Search10}}
 {id=WavePlayer_351, properties {soundWaveAssetPtr=&/Ammunition_Search12}}
 {id=WavePlayer_37, properties {soundWaveAssetPtr=/Game/Sounds/Foley_Loot/Containers/Misc/Prop_Gear_Search_05}}
 {id=WavePlayer_38, properties {soundWaveAssetPtr=/Game/Sounds/Foley_Loot/Containers/Misc/Prop_Gear_Search_01}}
 {id=WavePlayer_4, properties {soundWaveAssetPtr=&Ammunition_Search10}}
 {id=WavePlayer_41, properties {soundWaveAssetPtr=/Game/Sounds/Foley_Loot/Containers/Misc/Prop_Gear_Search_06}}
 {id=WavePlayer_6, properties {soundWaveAssetPtr=&Prop_Gear_Search_05}}
 {id=WavePlayer_7, properties {soundWaveAssetPtr=&Prop_Gear_Search_01}}
 {id=WavePlayer_9, properties {soundWaveAssetPtr=&ammobox_B_search_01}}
]
properties {
 attenuationSettings=/Game/Sounds/Attenuation/Fort_Prop_Search_Att
 bOverrideConcurrency=true
 maxDistance=1300
 soundClassObject=/Figment_S10/Sounds/FortSoundClassesAndModes/Fort_Searching_Foley
 volumeMultiplier=0.5
 concurrencyOverrides {resolutionRule=PreventNew}
}
